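; Creature Placer Creative device that previews a selectable creature type and configures its AI spawning.
; User options set the spawn group, game-phase activation, ranges and despawn rules.
; Creature type changes refresh the visuals, with an async-loaded idle animation, transition effects and editor workarounds for unfired load callbacks.
; Spawn/despawn events drive on/off handlers, spawned pawns get a type tag via GameplayEffect, and kills broadcast an eliminated event.
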(event EventBeginPlay
  (Utilities|FlowControl|Sequence
    (:then_0
      (Creative|Networking|Authority/Cosmetic
        (:Cosmetic
          (Variables|/Script/Engine.SkeletalMeshComponent|SetbNoSkeletonUpdate (Building|IsPreviewBuildingActor) (Variables|Debug|GetCreatureSkeletalMeshComponent)) ; Turning off skeleton updates while it's a preview fixes placement bounds
          (EventDispatchers|BindEventtoOnCreativeAssetsLoaded (Variables|BuildingProp_CreaturePlacer|GetAssetLoadingTrackingComponent) (EventDispatchers|CreateEvent :!event Custom|OnCreatureVisualsLoadCompleteClient)))))
    (:then_1
      (CallFunction|SetSpawnLocation (Variables|Default|GetSpawnPoint)))
    (:then_2
      (Visibility|SetVisibilityComponents (Variables|Default|GetMeshesVisibleInGameComponent) (Utilities|Array|MakeArray (Variables|Debug|GetStaticMeshComponent) (Variables|Debug|GetCreatureSkeletalMeshComponent)) true)
      (Visibility|SetVisibilityComponents (Variables|Default|GetVisibleInGameComponent) (Utilities|Array|MakeArray (Variables|Default|GetCreatureHighlight) (Variables|Default|GetHoloFloor) (Variables|Default|GetLightRaysA) (Variables|Default|GetLightRaysB) (Variables|Default|GetSpotLight) (Variables|Default|GetP_CreaturePlacer_Smoke)) true))))

(event EventMinigameStateChanged
  (params (Minigame :type Fort%20Minigame%20Object%20Reference) (NewMinigameState :type EFortMinigameState%20Enum :default "PreGame"))
  (if (Utilities|Enum|Equal(Enum) NewMinigameState "PreGame")
    (CallFunction|SetupConfigurables)
    (else
      _))
  (|CallParentFunction Minigame NewMinigameState))

(event OnAnyPropertyChanged(ToyOptionsComponent)
  (CallFunction|SetupConfigurables)
  (if (CallFunction|ShouldUpdateCreatureVisuals)
    (Variables|Internal|SetPrevAITypeIndex (Variables|Configurables|GetAITypePresetIndex)) ; Set this after checking if the PresetIndex changed
    (CallFunction|UpdateCreatureVisuals)
    (if (Editor|InEditor)
      (CallFunction|OnCreatureVisualsLoadCompleteClient) ; Code delegate does not bind/fire in editor, so manually call the completion callback
    )))

(event OnPropertiesLoaded(ToyOptionsComponent)
  (CallFunction|SetupConfigurables)
  (if (CallFunction|ShouldUpdateCreatureVisuals)
    (Variables|Internal|SetPrevAITypeIndex (Variables|Configurables|GetAITypePresetIndex)) ; Set this after checking if the PresetIndex changed
    (CallFunction|UpdateCreatureVisuals)
    (if (Editor|InEditor)
      (CallFunction|OnCreatureVisualsLoadCompleteClient) ; Code delegate does not bind/fire in editor, so manually call the completion callback
    )))

(event OnVisibilityChanged(MeshesVisibleInGameComponent)
  (params (Visible :type Boolean))
  (bind _creatureskeletalmeshcomponent (Variables|Debug|GetCreatureSkeletalMeshComponent))
  (bind _staticmeshcomponent (Variables|Debug|GetStaticMeshComponent))
  (if Visible
    (CallFunction|AdjustCollision false _creatureskeletalmeshcomponent)
    (CallFunction|AdjustCollision false _staticmeshcomponent)
    (else
      (CallFunction|AdjustCollision true _creatureskeletalmeshcomponent)
      (CallFunction|AdjustCollision true _staticmeshcomponent))))

(event Custom|OnCreatureVisualsLoadCompleteClient
  (bind _aitypepresetindex (Variables|Configurables|GetAITypePresetIndex))
  (bind _enemyidleanimations (Variables|Default|GetEnemyIdleAnimations))
  (bind _output (Utilities|Array|Get(acopy) _enemyidleanimations _aitypepresetindex))
  (bind _creatureskeletalmeshcomponent (Variables|Debug|GetCreatureSkeletalMeshComponent))
  ;[1 1 1] Client Visuals
  (if (Utilities|IsValidSoftObjectReference _output)
    (bind _loaded_object (Creative|Loading|AsyncLoadEditorSafe _output)
      (:Loaded
        (Utilities|FlowControl|Sequence
          (:then_0
            (bind _asanim_sequence (Utilities|Casting|CastToAnimSequence _loaded_object))
            (if (not (CallFunction|IsPlayingAnimation _asanim_sequence))
              (Components|Animation|PlayAnimation _creatureskeletalmeshcomponent _asanim_sequence (Variables|Default|GetLoopingPreviewAnimation))
              (Variables|/Script/Engine.SkeletalMeshComponent|SetAnimationData (Utilities|Struct|MakeSingleAnimationPlayData _asanim_sequence (Variables|Default|GetLoopingPreviewAnimation) (Components|Animation|IsPlaying (Variables|Debug|GetCreatureSkeletalMeshComponent)) (Components|Animation|GetPosition (Variables|Debug|GetCreatureSkeletalMeshComponent)) (Components|Animation|GetPlayRate (Variables|Debug|GetCreatureSkeletalMeshComponent))) _creatureskeletalmeshcomponent)
              ;[0.5117 1 0.5418] Play animation in editor
              (if (Editor|InEditor)
                (Components|Animation|SetAnimationMode _creatureskeletalmeshcomponent "AnimationSingleNode" true)
                (Components|Animation|SetAnimation _creatureskeletalmeshcomponent _asanim_sequence)
                (Components|SkeletalMesh|SetUpdateAnimationinEditor _creatureskeletalmeshcomponent true))))
          (:then_1
            (Effects|SpawnEmitterAttached :EmitterTemplate "/Game/Creative/Devices/CreaturePlacer/VFX/P_CreaturePlacer_Transition.P_CreaturePlacer_Transition" :AttachToComponent (Variables|Debug|GetCreatureSkeletalMeshComponent) :AttachPointName "Root" :Scale "1.000000,1.000000,1.000000" :LocationType "KeepRelativeOffset" :bAutoDestroy true :PoolingMethod "AutoRelease" :bAutoActivate true)
            (Audio|PlaySoundatLocation "/Game/Sounds/Creative/Toys/CreaturePlacer/CreaturePlacer_CreatureAppears_Cue.CreaturePlacer_CreatureAppears_Cue" (Transformation|GetActorLocation))))))))

(event EventBP_NotifyAIPawnSpawned
  (params (SpawnedAI :type Fort%20AIPawn%20Object%20Reference) (SpawnSlotLocation :type Vector%20%28by%20ref%29 :default "0, 0, 0"))
  (bind _aipawnabilitysystemcomponent (Variables|/Script/FortniteGame.FortAIPawn|GetAIPawnAbilitySystemComponent SpawnedAI))
  (Utilities|IsValid SpawnedAI
    (:Is%20Valid
      (bind _returnvalue (Ability|GameplayEffect|AddGrantedTag (GameplayEffects|MakeOutgoingSpec :self _aipawnabilitysystemcomponent :GameplayEffectClass "/Game/Creative/Devices/CreaturePlacer/GameplayEffects/GE_EliminationManagerAddTag.GE_EliminationManagerAddTag_C" :Level 0.0) (Variables|Default|GetSpawnAppliedGameplayTag)))
      (GameplayEffects|ApplyGameplayEffectSpecToSelf _aipawnabilitysystemcomponent _returnvalue))))

(event EventBP_NotifyAIPawnKilled
  (params (KilledAI :type Fort%20AIPawn%20Object%20Reference) (KillingInstigator :type Controller%20Object%20Reference))
  (GameplayEvent|BroadcastEvent (Variables|UserOptions|Events|GetOn%20Eliminated) 0 (Utilities|Struct|MakeGameplayEventCreatureEliminated KillingInstigator)))

(event Custom|SpawnEvent
  (params (Instigator :type Controller%20Object%20Reference))
  (CallFunction|OnTurnonReceived Instigator))

(event Custom|DespawnEvent
  (params (Instigator :type Controller%20Object%20Reference))
  (CallFunction|OnTurnOffReceived Instigator))

(fn GetCollisionStaticMesh
  (return "0, 0, 0" (Variables|/Script/Engine.StaticMeshComponent|GetStaticMesh (Variables|Debug|GetStaticMeshComponent))))

(fn GetMeshComponents
  (return 0))

(fn IgnoreWorldGridScaleOverride
  (return false))

(fn IsPlayingAnimation
  (params (AnimSequence :type Anim%20Sequence%20Object%20Reference))
  (return (and (Components|Animation|IsPlaying (Variables|Debug|GetCreatureSkeletalMeshComponent)) (== (Utilities|Struct|BreakSingleAnimationPlayData (Variables|/Script/Engine.SkeletalMeshComponent|GetAnimationData (Variables|Debug|GetCreatureSkeletalMeshComponent))) (Variables|GetAnimSequence)))))

(fn IsTargetableByLinearMover
  (return true))

(fn SetupConfigurables
  (Variables|Default|SetSpawnAppliedGameplayTag (Utilities|Map|Find (Variables|Default|GetEnemyTypeIntOptionMap) (Variables|Configurables|GetAITypePresetIndex)))
  (Variables|BuildingProp_CreaturePlacer|SetSelectedSpawnGroup (Utilities|Struct|BreakSpawnGroupVisuals (Utilities|Array|Get(aref) (Variables|BuildingProp_CreaturePlacer|GetSpawnGroupVisuals) (Variables|Configurables|GetAITypePresetIndex))))
  (Variables|BuildingProp_CreaturePlacer|SetbHideVFX (not (Variables|Configurables|GetbVFXVisible)))
  ;[1 1 1] Set the Spawn on Game State Logic
  (switch int (Variables|Configurables|GetActivateOnGamePhase)
    (:0
      (Variables|BuildingProp_CreaturePlacer|SetGameStateEnable "PreGame")
      (Variables|BuildingProp_CreaturePlacer|SetbEnabledOnGameState false))
    (:1
      (Variables|BuildingProp_CreaturePlacer|SetGameStateEnable "Warmup")
      (Variables|BuildingProp_CreaturePlacer|SetbEnabledOnGameState true))
    (:2
      (Variables|BuildingProp_CreaturePlacer|SetGameStateEnable "InProgress")))
  (Variables|BuildingProp_AISpawner|SetDespawnRangeOverride (* (Variables|Configurables|GetDespawnRangePresetIndex) 512.0))
  (Variables|BuildingProp_CreaturePlacer|SetbDestroyPreviousOnSpawn (not (Variables|Configurables|GetbSpawnIfNeeded)))
  (Variables|BuildingProp_AISpawner|SetbUseDistanceToAI (== (Variables|Configurables|GetDespawnTypePresetIndex) 0))
  (Variables|BuildingProp_AISpawner|SetbCanAIDespawnWhenTooFar (!= (Variables|Configurables|GetDespawnTypePresetIndex) 2))
  (if (>= (Variables|Configurables|GetActivationRangePresetIndex) 0)
    (Variables|BuildingProp_AISpawner|SetActivationRangeLevel (Math|Conversions|ToFloat(Integer) (+ (Variables|Configurables|GetActivationRangePresetIndex) 1)))
    (else
      (Variables|BuildingProp_AISpawner|SetActivationRangeLevel 5.0))))

(fn ShouldInvisiblePropsBeAllowed
  (return false))

(fn ShouldUpdateCreatureVisuals
  (return (!= (Variables|Internal|GetPrevAITypeIndex) (Variables|Configurables|GetAITypePresetIndex))))

(fn ConstructionScript)
