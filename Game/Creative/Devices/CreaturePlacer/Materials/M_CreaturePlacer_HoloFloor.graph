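# Additive unlit material for the Creature Placer device's holographic floor, drawing a soft circular glow.
# A SphereMask over the UVs, centered at the middle with zero hardness, forms a radial gradient reaching the edges.
# The mask is multiplied by itself to tighten the falloff, then tinted by the Color vector parameter, which defaults to deep blue.
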Material /Game/Creative/Devices/CreaturePlacer/Materials/M_CreaturePlacer_HoloFloor
graph {
 properties {blendMode=BLEND_Additive, pixelDepthOffsetMode=PDOM_Legacy, shadingModel=MSM_Unlit, usedShadingModels=Unlit}
 nodes [
  n1 = Constant() {r=0.5}
  n2 = SphereMask(A=TextureCoordinate(), B=@n1, Radius=@n1, Hardness=Constant())
  Multiply(A=Multiply(A=@n2, B=@n2), B=VectorParameter() {parameterName=Color, defaultValue=<rgba 0.0467 0.034 0.6337 0>}.RGB)
 ]
}
