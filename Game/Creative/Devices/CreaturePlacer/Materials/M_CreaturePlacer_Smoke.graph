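# Unlit translucent sprite material for the Creature Placer device's smoke effect, usable with Niagara and particle sprites.
# Opacity multiplies the red channel of an inky smoke flipbook texture by particle alpha, then passes through a depth fade with a parameterized distance to soften intersections.
# Particle color RGB feeds a shared FX emissive adjustments function.
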Material /Game/Creative/Devices/CreaturePlacer/Materials/M_CreaturePlacer_Smoke
graph {
 properties {
  bAutomaticallySetUsageInEditor=false
  bUsedWithNiagaraSprites=true
  bUsedWithParticleSprites=true
  blendMode=BLEND_Translucent
  pixelDepthOffsetMode=PDOM_Legacy
  shadingModel=MSM_Unlit
  usedShadingModels=Unlit
 }
 nodes [
  DepthFade(
   Opacity=Multiply(A=TextureSample() {texture=/Game/Effects/Fort_Effects/Textures/T_InkySmoke_Flipbook}.R, B=ParticleColor().A)
   FadeDistance=ScalarParameter() {defaultValue=10, parameterName=FadeDistance}
  )
  MaterialFunctionCall(Emissive=ParticleColor().RGB) {materialFunction=&MF_FX_EmissiveAdjustments}
 ]
}
